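# Physics asset for the Neon Cat Spy medium female player body, covering its dynamic hip-clip and leg flaps.
# Box bodies on the pelvis, both thighs and four dyn_ flap bones form a small lower-body setup.
# Hip-clip flaps hang from the pelvis with twist locked, and leg flaps from their thighs with angular limits; the right one is named dyn_regFlap_r.
# Both thighs join the pelvis with free swing and twist.
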PhysicsAsset /Game/Characters/Player/Female/Medium/Bodies/F_MED_Neon_Cat_Spy/Meshes/F_MED_Neon_Cat_Spy_Physics
bodies {
 dyn_hipClipFlap_l [{shapeType=Box, extentX=0.8837, extentY=6.5934, extentZ=4.0713, center=<3.3288 -0.0389 0.1863>, rotation=<rot -6.4254 -90 -11.7392>}]
 dyn_hipClipFlap_r [{shapeType=Box, extentX=0.8304, extentY=6.9269, extentZ=4.761, center=<3.0163 0.0517 0.1296>, rotation=<rot 22.1499 -90 -30.026>}]
 dyn_legFlap_l [{shapeType=Box, extentX=3.6116, extentY=10.9986, extentZ=4.0551, center=<2.8661 0 -0.0002>, rotation=<rot 0 -90 0>}]
 dyn_regFlap_r [{shapeType=Box, extentX=3.6116, extentY=10.9986, extentZ=4.0551, center=<3.8433 0 0>, rotation=<rot 0 -90 0>}]
 pelvis [{shapeType=Box, extentX=0.5003, extentY=3.2679, extentZ=6.2265, center=<-0.9882 -2.6046 -0.024>, rotation=<rot 1.2124 -90 0.4398>}]
 thigh_l [{shapeType=Box, extentX=3.6632, extentY=4.9731, extentZ=20.0185, center=<-26.1874 -1.7559 0.4305>, rotation=<rot 0 -86.7289 -89.0445>}]
 thigh_r [{shapeType=Box, extentX=3.6759, extentY=4.1823, extentZ=20.0359, center=<26.5952 1.8521 -0.3883>, rotation=<rot 0 93.9379 -86.8815>}]
}
constraints [
 {bone1Name=dyn_hipClipFlap_r, bone2Name=pelvis, swing1LimitDegrees=20, swing2LimitDegrees=8, twistLimitDegrees=45, twistMotion=Locked}
 {bone1Name=dyn_hipClipFlap_l, bone2Name=pelvis, swing1LimitDegrees=20, swing2LimitDegrees=8, twistLimitDegrees=45, twistMotion=Locked}
 {bone1Name=dyn_legFlap_l, bone2Name=thigh_l, swing1LimitDegrees=20, swing2LimitDegrees=50, twistLimitDegrees=15}
 {bone1Name=dyn_regFlap_r, bone2Name=thigh_r, swing1LimitDegrees=20, swing2LimitDegrees=50, twistLimitDegrees=15}
 {
  bone1Name=thigh_r
  bone2Name=pelvis
  swing1LimitDegrees=45
  swing1Motion=Free
  swing2LimitDegrees=45
  swing2Motion=Free
  twistLimitDegrees=45
  twistMotion=Free
 }
 {
  bone1Name=thigh_l
  bone2Name=pelvis
  swing1LimitDegrees=45
  swing1Motion=Free
  swing2LimitDegrees=45
  swing2Motion=Free
  twistLimitDegrees=45
  twistMotion=Free
 }
]
