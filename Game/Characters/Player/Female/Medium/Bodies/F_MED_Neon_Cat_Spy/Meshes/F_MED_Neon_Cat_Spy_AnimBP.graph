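# Post-body animation Blueprint for the F_MED_Neon_Cat_Spy female player body, adding secondary motion to its rings and hip flaps.
# It copies the attached parent mesh's pose, then weighted multi-bone constraints drive left chest and shoulder D-ring bases and position both hip clip flaps.
# A RigidBody node simulates the override physics asset in root-bone space with overridden gravity and damped component motion.
# A Control Rig applies the final pass; the event graph is empty.
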AnimBlueprint /Game/Characters/Player/Female/Medium/Bodies/F_MED_Neon_Cat_Spy/Meshes/F_MED_Neon_Cat_Spy_AnimBP
target_skeleton=&/F_MED_Neon_Cat_Spy_Skeleton
graphs [
 AnimGraph {nodes [
   n1 = "Constraint - dyn_dRingChest_Base_l (PPPPP)"!(
    ComponentPose=Local To Component(LocalPose=Copy Pose From Mesh() {bUseAttachedParent=true})
    Alpha=1
    ConstraintWeights_0=0.0003
    ConstraintWeights_1:float=0.2828
    ConstraintWeights_2:float=0.2798
    ConstraintWeights_3:float=0.1911
    ConstraintWeights_4:float=0.2459
   ) -> (Pose) {
    alphaBoolBlend {blendOption=Linear}
    boneToModify {boneName=dyn_dRingChest_Base_l}
    constraintSetup [
     {targetBone {boneName=clavicle_r}}
     {targetBone {boneName=upperarm_twist_01_l}}
     {targetBone {boneName=spine_04}}
     {targetBone {boneName=spine_05}}
     {targetBone {boneName=clavicle_l}}
    ]
    constraintWeights [1,1,1,1,1]
   }
   n2 = "Constraint - dyn_dRingShoulder_Base_l (PPPPPP)"!(
    ComponentPose=@n1
    Alpha=1
    ConstraintWeights_0=0.2071
    ConstraintWeights_1:float=0.0086
    ConstraintWeights_2:float=0.2433
    ConstraintWeights_3:float=0.2077
    ConstraintWeights_4:float=1.2736
    ConstraintWeights_5:float=0.0074
   ) -> (Pose) {
    alphaBoolBlend {blendOption=Linear}
    boneToModify {boneName=dyn_dRingShoulder_Base_l}
    constraintSetup [
     {targetBone {boneName=clavicle_l}}
     {targetBone {boneName=upperarm_twist_02_l}}
     {targetBone {boneName=spine_04}}
     {targetBone {boneName=spine_05}}
     {targetBone {boneName=upperarm_twist_01_l}}
     {targetBone {boneName=upperarm_l}}
    ]
    constraintWeights [1,1,1,1,1,1]
   }
   n3 = "Constraint - dyn_hipClipFlap_l (TTTT)"!(
    ComponentPose=@n2
    Alpha=1
    ConstraintWeights_0=0.6775
    ConstraintWeights_1:float=0.2845
    ConstraintWeights_2:float=0.0061
    ConstraintWeights_3:float=0.0319
   ) -> (Pose) {
    alphaBoolBlend {blendOption=Linear}
    boneToModify {boneName=dyn_hipClipFlap_l}
    constraintSetup [
     {transformType=Translation, targetBone {boneName=pelvis}}
     {transformType=Translation, targetBone {boneName=thigh_twist_01_l}}
     {transformType=Translation, targetBone {boneName=dyn_thighVolume_l}}
     {transformType=Translation, targetBone {boneName=thigh_l}}
    ]
    constraintWeights [1,1,1,1]
   }
   n4 = "Constraint - dyn_hipClipFlap_r (TTTT)"!(
    ComponentPose=@n3
    Alpha=1
    ConstraintWeights_0=0.6775
    ConstraintWeights_1:float=0.2845
    ConstraintWeights_2:float=0.0061
    ConstraintWeights_3:float=0.0319
   ) -> (Pose) {
    alphaBoolBlend {blendOption=Linear}
    boneToModify {boneName=dyn_hipClipFlap_r}
    constraintSetup [
     {transformType=Translation, targetBone {boneName=pelvis}}
     {transformType=Translation, targetBone {boneName=thigh_twist_01_r}}
     {transformType=Translation, targetBone {boneName=dyn_thighVolume_r}}
     {transformType=Translation, targetBone {boneName=thigh_r}}
    ]
    constraintWeights [1,1,1,1]
   }
   n5 = RigidBody(ComponentPose=@n4) {
    bClampLinearTranslationLimitToRefPose=true
    bOverrideWorldGravity=true
    overridePhysicsAsset=&/F_MED_Neon_Cat_Spy_Physics
    simulationSpace=BaseBoneSpace
    alphaBoolBlend {blendOption=Linear}
    baseBoneRef {boneName=root}
    componentAppliedLinearAccClamp=<4000 4000 4000>
    componentLinearAccScale=<0.4 0.4 0.4>
    componentLinearVelScale=<0.2 0.2 0.15>
    overrideWorldGravity {z=-980}
   }
   n6 = Control Rig(Source=Component To Local(ComponentPose=@n5)) {
    alphaBoolBlend {blendOption=Linear}
    controlRigAssetReference {blueprintRigClass=&/F_MED_NeonCatSpy_CtrlRig.F_MED_NeonCatSpy_CtrlRig_C}
   }
   Output Pose(Result=@n6) {layerGroup=None}
  ]}
 EventGraph {nodes [Event BlueprintUpdateAnimation()]}
]
variables []
