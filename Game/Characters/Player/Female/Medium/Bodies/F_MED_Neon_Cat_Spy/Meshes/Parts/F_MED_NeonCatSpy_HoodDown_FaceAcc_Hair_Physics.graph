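# Physics asset for the hair on the Neon Cat Spy female character's hood-down head accessory, defining bodies for simulated hair locks.
# Main chain A runs from a box base at the head through four capsules, with branches D and E off A_02, and B and C off A_03.
# Hair joints have tight swing and twist limits.
# Head, clavicle and upper-spine capsules supply body shapes, linked, like the hair base, by free angular constraints.
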PhysicsAsset /Game/Characters/Player/Female/Medium/Bodies/F_MED_Neon_Cat_Spy/Meshes/Parts/F_MED_NeonCatSpy_HoodDown_FaceAcc_Hair_Physics
bodies {
 clavicle_l [{shapeType=Capsule, length=11.7179, radius=4.4982, center=<2.5623 -1.8369 0>, rotation=<rot 90 0 180>}]
 clavicle_r [{shapeType=Capsule, length=11.7179, radius=4.4982, center=<-2.5623 1.8369 0>, rotation=<rot 90 0 0>}]
 dyn_hairLock_A_01 [{shapeType=Capsule, length=5.2756, radius=1.3386, center=<1.4128 -0.7065 -0.2345>, rotation=<rot 90 0 0>}]
 dyn_hairLock_A_01_Base [{shapeType=Box, extentX=1.01, extentY=1.01, extentZ=1.01, rotation=<rot 0 -90 0>}]
 dyn_hairLock_A_02 [{shapeType=Capsule, length=4.7983, radius=1.4992, center=<2.0513 -0.7993 -0.0001>, rotation=<rot 90 0 -6.2762>}]
 dyn_hairLock_A_03 [{shapeType=Capsule, length=5.3195, radius=1.5806, center=<1.928 -0.5542 -0.0116>, rotation=<rot 0 87.9828 -89.2722>}]
 dyn_hairLock_A_04 [{shapeType=Capsule, length=4.8571, radius=0.7138, center=<2.6506 -0.1603 0.0439>, rotation=<rot 0 92.5338 -89.5083>}]
 dyn_hairLock_B_01 [{shapeType=Capsule, length=2.5354, radius=0.5, center=<1.315 -0.0527 0.021>, rotation=<rot 0 -87.9047 -88.0779>}]
 dyn_hairLock_C_01 [{shapeType=Capsule, length=3.3276, radius=0.5, center=<1.6208 -0.0002 0>, rotation=<rot 90 0 0>}]
 dyn_hairLock_D_01 [{shapeType=Capsule, length=2.366, radius=0.5, center=<1.178 0 0>, rotation=<rot 90 0 0>}]
 dyn_hairLock_D_02 [{shapeType=Capsule, length=2.0952, radius=0.505, center=<0.9536 -0.025 0.0271>, rotation=<rot 0 -91.1842 -89.9694>}]
 dyn_hairLock_E_01 [{shapeType=Capsule, length=2.1617, radius=0.505, center=<1.2064 -0.0001 0>, rotation=<rot 90 0 0>}]
 head [{shapeType=Capsule, length=3.1792, radius=3.3306, center=<0.4391 2.6051 0>, rotation=<rot 90 0 -21.3102>}]
 spine_05 [{shapeType=Capsule, length=17.1589, radius=8.1499}]
}
constraints [
 {
  bone1Name=clavicle_r
  bone2Name=spine_05
  swing1LimitDegrees=45
  swing1Motion=Free
  swing2LimitDegrees=45
  swing2Motion=Free
  twistLimitDegrees=45
  twistMotion=Free
 }
 {
  bone1Name=clavicle_l
  bone2Name=spine_05
  swing1LimitDegrees=45
  swing1Motion=Free
  swing2LimitDegrees=45
  swing2Motion=Free
  twistLimitDegrees=45
  twistMotion=Free
 }
 {
  bone1Name=dyn_hairLock_A_01_Base
  bone2Name=head
  swing1LimitDegrees=45
  swing1Motion=Free
  swing2LimitDegrees=45
  swing2Motion=Free
  twistLimitDegrees=45
  twistMotion=Free
 }
 {bone1Name=dyn_hairLock_A_01, bone2Name=dyn_hairLock_A_01_Base, swing1LimitDegrees=35, swing2LimitDegrees=35, twistLimitDegrees=8}
 {bone1Name=dyn_hairLock_A_02, bone2Name=dyn_hairLock_A_01, swing1LimitDegrees=30, swing2LimitDegrees=30, twistLimitDegrees=4}
 {bone1Name=dyn_hairLock_A_03, bone2Name=dyn_hairLock_A_02, swing1LimitDegrees=10, swing2LimitDegrees=10, twistLimitDegrees=6}
 {bone1Name=dyn_hairLock_A_04, bone2Name=dyn_hairLock_A_03, swing1LimitDegrees=16, swing2LimitDegrees=16, twistLimitDegrees=8}
 {bone1Name=dyn_hairLock_B_01, bone2Name=dyn_hairLock_A_03, swing1LimitDegrees=12, swing2LimitDegrees=12, twistLimitDegrees=12}
 {bone1Name=dyn_hairLock_D_01, bone2Name=dyn_hairLock_A_02, swing1LimitDegrees=16, swing2LimitDegrees=16, twistLimitDegrees=8}
 {bone1Name=dyn_hairLock_D_02, bone2Name=dyn_hairLock_D_01, swing1LimitDegrees=16, swing2LimitDegrees=16, twistLimitDegrees=8}
 {bone1Name=dyn_hairLock_C_01, bone2Name=dyn_hairLock_A_03, swing1LimitDegrees=20, swing2LimitDegrees=20, twistLimitDegrees=8}
 {bone1Name=dyn_hairLock_E_01, bone2Name=dyn_hairLock_A_02, swing1LimitDegrees=12, swing2LimitDegrees=12, twistLimitDegrees=8}
 {
  bone1Name=head
  bone2Name=spine_05
  swing1LimitDegrees=45
  swing1Motion=Free
  swing2LimitDegrees=45
  swing2Motion=Free
  twistLimitDegrees=45
  twistMotion=Free
 }
]
