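# Physics asset for the Neon Cat Spy female player face accessory, defining secondary motion bodies for the cat ears and hood.
# A head capsule anchors small base boxes for the ears, hood sides and hood top through angularly free constraints.
# Two-segment ear chains and single hood flaps hang off those bases with tight swing and twist limits.
# The hood back-push pair is angularly free and has no constraint to the head.
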PhysicsAsset /Game/Characters/Player/Female/Medium/Bodies/F_MED_Neon_Cat_Spy/Meshes/Parts/F_MED_Neon_Cat_Spy_FaceAcc_Physics
bodies {
 dyn_ear_01_Base_l [{shapeType=Box, extentX=1.01, extentY=1.01, extentZ=1.01, rotation=<rot 0 -90 0>}]
 dyn_ear_01_Base_r [{shapeType=Box, extentX=1.01, extentY=1.01, extentZ=1.01, rotation=<rot 0 -90 0>}]
 dyn_ear_01_l [{shapeType=Box, extentX=2.6555, extentY=4.8098, extentZ=5.3988, center=<1.8541 -0.1145 -0.2871>, rotation=<rot -0.0399 -89.9998 -2.7613>}]
 dyn_ear_01_r [{shapeType=Box, extentX=2.7632, extentY=4.4362, extentZ=5.2581, center=<1.7599 -0.22 -0.0545>, rotation=<rot -17.3563 -90 3.1613>}]
 dyn_ear_02_l [{shapeType=Box, extentX=1.866, extentY=3.6589, extentZ=4.212, center=<1.0948 -0.1732 0.3629>, rotation=<rot 0 83.0643 -87.6404>}]
 dyn_ear_02_r [{shapeType=Box, extentX=1.9261, extentY=3.3332, extentZ=4.0258, center=<2.0201 -0.1196 -0.3492>, rotation=<rot -9.4646 -95.4834 -86.9511>}]
 dyn_hoodBackPush_01 [{shapeType=Box, extentX=1.01, extentY=1.01, extentZ=1.01, rotation=<rot 0 -90 0>}]
 dyn_hoodBackPush_01_Base [{shapeType=Box, extentX=1.01, extentY=1.01, extentZ=1.01, rotation=<rot 0 -90 0>}]
 dyn_hoodTop [{shapeType=Box, extentX=1.3538, extentY=4.491, extentZ=6.4887, center=<1.9187 0.101 0.0464>, rotation=<rot -6.3454 -89.9997 -0.3508>}]
 dyn_hoodTop_Base [{shapeType=Box, extentX=1.01, extentY=1.01, extentZ=1.01, rotation=<rot 0 -90 0>}]
 dyn_hood_Base_l [{shapeType=Box, extentX=1.01, extentY=1.01, extentZ=1.01, rotation=<rot 0 -90 0>}]
 dyn_hood_Base_r [{shapeType=Box, extentX=1.01, extentY=1.01, extentZ=1.01, rotation=<rot 0 -90 0>}]
 dyn_hood_l [{shapeType=Box, extentX=4.7024, extentY=7.2407, extentZ=1.01, center=<3 0 0>, rotation=<rot -5 -90 0>}]
 dyn_hood_r [{shapeType=Box, extentX=4.7024, extentY=7.2407, extentZ=1.01, center=<3 0 0>, rotation=<rot 5 -89.9999 0>}]
 head [{shapeType=Capsule, length=13.3903, radius=6.9018, rotation=<rot 90 -90 -110.0018>}]
}
constraints [
 {bone1Name=dyn_hood_r, bone2Name=dyn_hood_Base_r, swing1LimitDegrees=4, swing2LimitDegrees=20, twistLimitDegrees=4}
 {bone1Name=dyn_hood_l, bone2Name=dyn_hood_Base_l, swing1LimitDegrees=4, swing2LimitDegrees=20, twistLimitDegrees=4}
 {bone1Name=dyn_ear_01_r, bone2Name=dyn_ear_01_Base_r, swing1LimitDegrees=18, swing2LimitDegrees=16, twistLimitDegrees=6}
 {bone1Name=dyn_ear_02_r, bone2Name=dyn_ear_01_r, swing1LimitDegrees=16, swing2LimitDegrees=12, twistLimitDegrees=8}
 {bone1Name=dyn_ear_01_l, bone2Name=dyn_ear_01_Base_l, swing1LimitDegrees=12, swing2LimitDegrees=10, twistLimitDegrees=6}
 {bone1Name=dyn_ear_02_l, bone2Name=dyn_ear_01_l, swing1LimitDegrees=14, swing2LimitDegrees=10, twistLimitDegrees=8}
 {bone1Name=dyn_hoodTop, bone2Name=dyn_hoodTop_Base, swing1LimitDegrees=16, swing2LimitDegrees=4, twistLimitDegrees=4}
 {
  bone1Name=dyn_hoodBackPush_01
  bone2Name=dyn_hoodBackPush_01_Base
  swing1LimitDegrees=45
  swing1Motion=Free
  swing2LimitDegrees=45
  swing2Motion=Free
  twistLimitDegrees=45
  twistMotion=Free
 }
 {
  bone1Name=dyn_hood_Base_r
  bone2Name=head
  swing1LimitDegrees=45
  swing1Motion=Free
  swing2LimitDegrees=45
  swing2Motion=Free
  twistLimitDegrees=45
  twistMotion=Free
 }
 {
  bone1Name=dyn_hood_Base_l
  bone2Name=head
  swing1LimitDegrees=45
  swing1Motion=Free
  swing2LimitDegrees=45
  swing2Motion=Free
  twistLimitDegrees=45
  twistMotion=Free
 }
 {
  bone1Name=dyn_ear_01_Base_r
  bone2Name=head
  swing1LimitDegrees=45
  swing1Motion=Free
  swing2LimitDegrees=45
  swing2Motion=Free
  twistLimitDegrees=45
  twistMotion=Free
 }
 {
  bone1Name=dyn_ear_01_Base_l
  bone2Name=head
  swing1LimitDegrees=45
  swing1Motion=Free
  swing2LimitDegrees=45
  swing2Motion=Free
  twistLimitDegrees=45
  twistMotion=Free
 }
 {
  bone1Name=dyn_hoodTop_Base
  bone2Name=head
  swing1LimitDegrees=45
  swing1Motion=Free
  swing2LimitDegrees=45
  swing2Motion=Free
  twistLimitDegrees=45
  twistMotion=Free
 }
]
